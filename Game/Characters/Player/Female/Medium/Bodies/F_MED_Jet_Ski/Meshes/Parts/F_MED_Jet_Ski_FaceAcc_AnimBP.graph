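# Post-copy physics AnimBP for the Jet Ski female character's face accessory mesh, adding secondary motion to the copied body pose.
# The live graph copies the attached parent mesh's pose, then runs a RigidBody node with the FaceAcc physics asset, simulated relative to spine_05 under overridden gravity.
# A chain of AnimDynamics nodes for dyn_hair bones remains in the graph but does not reach the output pose.
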AnimBlueprint /Game/Characters/Player/Female/Medium/Bodies/F_MED_Jet_Ski/Meshes/Parts/F_MED_Jet_Ski_FaceAcc_AnimBP
target_skeleton=&/F_MED_Jet_Ski_FaceAcc_Skeleton
graphs [
 AnimGraph {nodes [
   n1 = Anim Dynamics - Bone: dyn_hair_01(Alpha=1) {
    alpha=0.9
    angularDampingOverride=1
    angularSpringConstant=1500
    bAngularSpring=true
    bGravityOverrideInSimSpace=true
    bOverrideAngularDamping=true
    gravityScale=0.65
    numSolverIterationsPostUpdate=4
    numSolverIterationsPreUpdate=16
    simulationSpace=RootRelative
    alphaBoolBlend {blendOption=Linear}
    boundBone {boneName=dyn_hair_01}
    componentLinearAccScale=<1 1 1>
    componentLinearVelScale=<1 1 1>
    physicsBodyDefinitions [{
      sphereCollisionRadius=0
      boundBone {boneName=dyn_hair_01}
      boxExtents=<15 15 2.5>
      constraintSetup {angularTargetAxis=AxisZ, twistAxis=AxisZ, angularLimitsMax=<15 15 15>, angularLimitsMin=<-15 -15 -15>, angularTarget=<0 0 1>}
      localJointOffset=<0 0 -2>
     }]
   }
   n2 = Anim Dynamics - Bone: dyn_hair_02(ComponentPose=@n1, Alpha=1) {
    alpha=0.9
    angularDampingOverride=1
    angularSpringConstant=1500
    bAngularSpring=true
    bGravityOverrideInSimSpace=true
    bOverrideAngularDamping=true
    gravityScale=0.65
    numSolverIterationsPostUpdate=4
    numSolverIterationsPreUpdate=16
    simulationSpace=RootRelative
    alphaBoolBlend {blendOption=Linear}
    boundBone {boneName=dyn_hair_02}
    componentLinearAccScale=<1 1 1>
    componentLinearVelScale=<1 1 1>
    physicsBodyDefinitions [{
      sphereCollisionRadius=0
      boundBone {boneName=dyn_hair_02}
      boxExtents=<15 15 2.5>
      constraintSetup {angularTargetAxis=AxisZ, twistAxis=AxisZ, angularLimitsMax=<15 15 15>, angularLimitsMin=<-15 -15 -15>, angularTarget=<0 0 1>}
      localJointOffset=<0 0 -2>
     }]
   }
   n3 = Anim Dynamics - Bone: dyn_hair_03(ComponentPose=@n2, Alpha=1, ExternalForce="0,0,0") {
    alpha=0.9
    angularDampingOverride=1
    angularSpringConstant=1500
    bAngularSpring=true
    bGravityOverrideInSimSpace=true
    bOverrideAngularDamping=true
    gravityScale=0.65
    numSolverIterationsPostUpdate=4
    numSolverIterationsPreUpdate=16
    simulationSpace=RootRelative
    alphaBoolBlend {blendOption=Linear}
    boundBone {boneName=dyn_hair_03}
    componentLinearAccScale=<1 1 1>
    componentLinearVelScale=<1 1 1>
    physicsBodyDefinitions [{
      sphereCollisionRadius=0
      boundBone {boneName=dyn_hair_03}
      boxExtents=<15 15 2.5>
      constraintSetup {angularTargetAxis=AxisZ, twistAxis=AxisZ, angularLimitsMax=<15 15 15>, angularLimitsMin=<-15 -15 -15>, angularTarget=<0 0 1>}
      localJointOffset=<0 0 -2>
     }]
   }
   n4 = Anim Dynamics - Bone: dyn_hair_04(ComponentPose=@n3, Alpha=1, ExternalForce="0,0,0") {
    alpha=0.9
    angularDampingOverride=1
    angularSpringConstant=1500
    bAngularSpring=true
    bGravityOverrideInSimSpace=true
    bOverrideAngularDamping=true
    gravityScale=0.65
    numSolverIterationsPostUpdate=4
    numSolverIterationsPreUpdate=16
    simulationSpace=RootRelative
    alphaBoolBlend {blendOption=Linear}
    boundBone {boneName=dyn_hair_04}
    componentLinearAccScale=<1 1 1>
    componentLinearVelScale=<1 1 1>
    physicsBodyDefinitions [{
      sphereCollisionRadius=0
      boundBone {boneName=dyn_hair_04}
      boxExtents=<15 15 2.5>
      constraintSetup {angularTargetAxis=AxisZ, twistAxis=AxisZ, angularLimitsMax=<15 15 15>, angularLimitsMin=<-15 -15 -15>, angularTarget=<0 0 1>}
      localJointOffset=<0 0 -2>
     }]
   }
   n5 = Anim Dynamics - Bone: dyn_hair_05(ComponentPose=@n4, Alpha=1, ExternalForce="0,0,0") {
    alpha=0.9
    angularDampingOverride=1
    angularSpringConstant=1500
    bAngularSpring=true
    bGravityOverrideInSimSpace=true
    bOverrideAngularDamping=true
    gravityScale=0.65
    numSolverIterationsPostUpdate=4
    numSolverIterationsPreUpdate=16
    simulationSpace=RootRelative
    alphaBoolBlend {blendOption=Linear}
    boundBone {boneName=dyn_hair_05}
    componentLinearAccScale=<1 1 1>
    componentLinearVelScale=<1 1 1>
    physicsBodyDefinitions [{
      sphereCollisionRadius=0
      boundBone {boneName=dyn_hair_05}
      boxExtents=<15 15 2.5>
      constraintSetup {angularTargetAxis=AxisZ, twistAxis=AxisZ, angularLimitsMax=<15 15 15>, angularLimitsMin=<-15 -15 -15>, angularTarget=<0 0 1>}
      localJointOffset=<0 0 -2>
     }]
   }
   n6 = Anim Dynamics - Bone: dyn_hair_07(ComponentPose=@n5, Alpha=1, ExternalForce="0,0,0") {
    angularBiasOverride=1
    angularDampingOverride=1
    bGravityOverrideInSimSpace=true
    bOverrideAngularBias=true
    bOverrideAngularDamping=true
    gravityScale=0.1
    simulationSpace=RootRelative
    alphaBoolBlend {blendOption=Linear}
    boundBone {boneName=dyn_hair_07}
    componentLinearAccScale=<1 1 1>
    componentLinearVelScale=<1 1 1>
    physicsBodyDefinitions [{
      sphereCollisionRadius=0
      boundBone {boneName=dyn_hair_07}
      boxExtents=<5 5 5>
      constraintSetup {
       angularTargetAxis=AxisY
       twistAxis=AxisY
       angularLimitsMax=<10 0 10>
       angularLimitsMin=<-10 0 -10>
       angularTarget=<0 0 0>
       linearAxesMax=<0 0 3>
      }
      localJointOffset=<0 0 -2>
     }]
   }
   n7 = Anim Dynamics - Bone: dyn_hair_08(ComponentPose=@n6, Alpha=1, ExternalForce="0,0,0") {
    angularBiasOverride=1
    angularDampingOverride=1
    bGravityOverrideInSimSpace=true
    bOverrideAngularBias=true
    bOverrideAngularDamping=true
    gravityScale=0.1
    simulationSpace=RootRelative
    alphaBoolBlend {blendOption=Linear}
    boundBone {boneName=dyn_hair_08}
    componentLinearAccScale=<1 1 1>
    componentLinearVelScale=<1 1 1>
    physicsBodyDefinitions [{
      sphereCollisionRadius=0
      boundBone {boneName=dyn_hair_08}
      boxExtents=<5 5 5>
      constraintSetup {
       angularTargetAxis=AxisY
       twistAxis=AxisY
       angularLimitsMax=<10 0 10>
       angularLimitsMin=<-10 0 -10>
       angularTarget=<0 0 0>
       linearAxesMax=<0 0 3>
      }
      localJointOffset=<0 0 -2>
     }]
   }
   Anim Dynamics - Bone: dyn_hair_09(ComponentPose=@n7, Alpha=1, ExternalForce="0,0,0") {
    alpha=0.9
    angularDampingOverride=1
    angularSpringConstant=1500
    bAngularSpring=true
    bGravityOverrideInSimSpace=true
    bOverrideAngularDamping=true
    numSolverIterationsPostUpdate=4
    numSolverIterationsPreUpdate=16
    simulationSpace=RootRelative
    alphaBoolBlend {blendOption=Linear}
    boundBone {boneName=dyn_hair_09}
    componentLinearAccScale=<1 1 1>
    componentLinearVelScale=<1 1 1>
    physicsBodyDefinitions [{
      sphereCollisionRadius=0
      boundBone {boneName=dyn_hair_09}
      boxExtents=<15 15 2.5>
      constraintSetup {angularTargetAxis=AxisZ, twistAxis=AxisY, angularLimitsMax=<15 15 15>, angularLimitsMin=<-15 -15 -15>, angularTarget=<0 0 1>}
      localJointOffset=<0 0 -2.5>
     }]
   }
   n8 = Get mesh_to_copy() -> (mesh_to_copy:Skeletal Mesh Component Object Reference)
   n9 = RigidBody(ComponentPose=Local To Component(LocalPose=Copy Pose From Mesh(SourceMeshComponent=@n8) {bUseAttachedParent=true})) {
    bClampLinearTranslationLimitToRefPose=true
    bOverrideWorldGravity=true
    overridePhysicsAsset=&/F_MED_Jet_Ski_FaceAcc_Physics
    simulationSpace=BaseBoneSpace
    alphaBoolBlend {blendOption=Linear}
    baseBoneRef {boneName=spine_05}
    componentLinearAccScale=<0.4 0.35 0.35>
    componentLinearVelScale=<0.4 0.35 0.35>
    overrideWorldGravity {z=-980}
   }
   Output Pose(Result=Component To Local(ComponentPose=@n9)) {layerGroup=None}
  ]}
 EventGraph {nodes [
   Event BlueprintInitializeAnimation()
   > Set mesh_to_copy(mesh_to_copy:Skeletal Mesh Component Object Reference=GetOwnerSkeletalMeshComponent()) -> (Output_Get:Skeletal Mesh Component Object Reference)
   TryGetPawnOwner()
   Event BlueprintUpdateAnimation()
  ]}
]
variables [mesh_to_copy {type=SkeletalMeshComponent}]
